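# Master two-sided subsurface material for Asteria grass and flowers, with static switches for 2D masked variants.
# World position offset combines wind, player push and physics-field bending around per-blade pivots stored in UV1, using lookup-texture or true-rotation animation.
# Color draws on biome masks, landscape RVT, per-blade hue variation and optional snow tint, with camera-distance fade masks.
# Quality and shading-path switches substitute cheaper animation and shading on low-end and mobile.
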Material /Game/Environments/Asteria/Foliage/Grass/MasterMaterial/M_Asteria_Grass_BaseMat
graph {
 properties {
  bTangentSpaceNormal=false
  bUseMaterialAttributes=true
  bUsedWithInstancedStaticMeshes=true
  bUsedWithNanite=true
  pixelDepthOffsetMode=PDOM_Legacy
  shadingModel=MSM_Subsurface
  twoSided=true
  usedShadingModels=Subsurface
 }
 nodes [
  n1 = VertexColor()
  Add(A=@n1.G, B=PerInstanceRandom())
  n2 = LinearInterpolate(
   A=ScalarParameter() {defaultValue=-50, group=WPO-DistanceBased, parameterName=SinkIntoGroundAmount}
   B=Constant()
   Alpha=PerInstanceFadeAmount()
  )
  n3 = Constant3Vector() {constant=<rgba 0 0 1 0>}
  n4 = MaterialFunctionCall(Color=NamedRerouteUsage()) {materialFunction=&SplitComponents}
  n5 = CameraPositionWS()
  n6 = NamedRerouteUsage()
  n7 = CollectionParameter() {collection=&FortniteMaterialParameters, parameterName=GrassPushDownRadius}
  n8 = Constant()
  n9 = CollectionParameter() {collection=&FortniteMaterialParameters, parameterName=IsPlayerSliding}
  n10 = LinearInterpolate(A=Multiply(A=SphereMask(A=@n5, B=@n6, Radius=@n7, Hardness=@n8), B=Constant() {r=-200}), B=Constant(), Alpha=@n9)
  n11 = LinearInterpolate(
   A=MaterialFunctionCall(X=@n4.R, Y=@n4.R, Z=Add(A=@n4.B, B=@n10)) {materialFunction=&MakeFloat3}.Result
   Alpha=MaterialFunctionCall!(
    FadeBlendDistance=ScalarParameter() {defaultValue=500, group=WPO-DistanceBased, parameterName=WPO_FadeDistance, sortPriority=6}
    AutoHandleBadDistance
   ) {materialFunction=/Game/Global/MaterialFunctions/TechArt/MF_WPODisableBlend}.VertexShaderOutput
  ) {constB=0}
  n12 = StaticSwitchParameter(True=Add(A=@n11, B=Multiply(A=@n2, B=@n3)), False=@n11) {parameterName=Is2DMaskMaterial}
  n13 = NamedRerouteUsage()
  n14 = Add(A=WorldPosition() {worldPositionShaderOffset=WPT_ExcludeAllShaderOffsets}.XYZ, B=@n12)
  n15 = MaterialFunctionCall!(
   UseHeightLerp
   UV_WorldMask
   Use_ExtraSnowMask
   LandscapeBlend_Height
   Landscape_RockMask
   Snow_Mask
   Simple_UseWorldZNormal
   Simple_WorldNormal
   WorldPosition
  ) {materialFunction=/Game/Athena/Environments/Landscape/MaterialFunctions/Snow/MF_Athena_Holiday_SnowFall_S7}
  n16 = OneMinus(None=Power(Base=OneMinus(None=@n15.SnowMask)) {constExponent=4})
  Add(A=@n12, B=Multiply(A=Subtract(A=@n13, B=@n14), B=@n16))
  n17 = NamedRerouteDeclaration(None=LinearInterpolate(
    A=ScalarParameter() {defaultValue=0.16, group=Specular, parameterName=Specular_Near}
    B=ScalarParameter() {defaultValue=0.05, parameterName=Specular_Fade}
    Alpha=NamedRerouteUsage()
   )) {name=SpecularLOWQuality, nodeColor=<rgba 0.6641 0 1 1>}
  n18 = StaticSwitchParameter(True=Multiply(A=NamedRerouteUsage(), B=@n17), False=@n17) {group=Textures, parameterName=UseTextures}
  LinearInterpolate(
   A=@n18
   B=ScalarParameter() {defaultValue=0.1, desc=Optimization RVT, group=RVTOptimization, parameterName=RVT_Specluar_Constant}
   Alpha=NamedRerouteUsage()
  )
  n19 = NamedRerouteDeclaration(None=LinearInterpolate(
    A=ScalarParameter() {defaultValue=0.75, group=Roughness, parameterName=Roughness_Near, sliderMin=1}
    B=ScalarParameter() {defaultValue=2, group=Roughness, parameterName=Roughness_Far}
    Alpha=NamedRerouteUsage()
   )) {name=RoughnessLowDetail, nodeColor=<rgba 0.1484 0 1 1>}
  n20 = StaticSwitchParameter(True=Multiply(A=NamedRerouteUsage(), B=@n19), False=@n19) {group=Textures, parameterName=UseTextures}
  LinearInterpolate(
   A=@n20
   B=ScalarParameter() {defaultValue=0.688, desc=Optimization RVT, group=RVTOptimization, parameterName=RVT_Roughness_Constant}
   Alpha=NamedRerouteUsage()
  )
  n21 = NamedRerouteUsage()
  n22 = CollectionParameter() {collection=&FortniteMaterialParameters, parameterName=PlayerPosition}
  n23 = ComponentMask(None=@n22) {b=true, g=true, r=true}
  n24 = Normalize(VectorInput=Subtract(A=@n21, B=@n23))
  n25 = Saturate(None=SphereMask(A=@n23, B=@n21, Radius=Constant() {r=200}, Hardness=Constant()) {attenuationRadius=64, hardnessPercent=20}) {desc=PlayerSphereMask}
  LinearInterpolate(
   A=NamedRerouteUsage()
   B=Multiply(A=VectorParameter() {parameterName=PushDistance, defaultValue=<rgba -22.5 -22.5 -50 1>}.RGB, B=@n24)
   Alpha=@n25
  )
  n26 = VertexColor()
  LinearInterpolate(A=ScalarParameter() {group=Color, parameterName=HueBottom}, B=ScalarParameter() {group=Color, parameterName=HueTop}, Alpha=@n26.B)
  MaterialFunctionCall(
   In=VertexColor() {desc=only blend To RVT at the bottom So you can have green grass on sand}.B
   Contrast=ScalarParameter() {defaultValue=0.2, group=RVT, parameterName=NonRVTBlend}
  ) {materialFunction=&CheapContrast}
  n27 = MakeMaterialAttributes!(
   BaseColor=VectorParameter() {group=Color, parameterName=MobileGrassBaseColor, defaultValue=<rgba 0.3137 1 0.2252 1>}.RGB
   Metallic
   Specular
   Roughness
   Anisotropy
   EmissiveColor
   Opacity
   OpacityMask
   Normal
   Tangent
   WorldPositionOffset
   SubsurfaceColor
   ClearCoat
   ClearCoatRoughness
   AmbientOcclusion
   Refraction
   CustomizedUVs_0
   CustomizedUVs_1
   CustomizedUVs_2
   CustomizedUVs_3
   CustomizedUVs_4
   CustomizedUVs_5
   CustomizedUVs_6
   CustomizedUVs_7
   PixelDepthOffset
   ShadingModel
   Displacement
  )
  MaterialFunctionCall!(SetMaterialAttributes=@n27) {materialFunction=/Game/Athena/Helios/Environments/Landscape/Materials/MaterialFunctions_Utilities/MF_Athena_Grass_Biome_Colors}
  MaterialFunctionCall!(
   Previous WPO Value
   Force Magnitude Mask
   Force Magnitude Multiplier
   Project Field Force Vector to Normals
   Noise Size
   Noise Pan Speed
   Noise Falloff
   Stretch Noise Coordinates Z
  ) {desc=Sample global physics field add to WPO, materialFunction=/Game/Effects/Niagara/MasterMaterials/Functions/MF_FX_PhysicsFields_Grass}
  n28 = MaterialFunctionCall!(UseLandscapeCoords) {materialFunction=/Game/Athena/Asteria/Environments/Landscape/Materials/MaterialFunctions/MF_AsteriaTerrainMasks}
  MaterialFunctionCall!(BiomeMask-RGBA=@n28.BiomeMaskRGBA, TextureSamplingPosition=NamedRerouteUsage(), IsGrass=StaticBool() {value=true}) {materialFunction=/Game/Athena/Asteria/Environments/Landscape/Materials/MaterialFunctions/MF_AsteriaGrassColors}
  n29 = SetMaterialAttributes!(
   MaterialAttributes
   Base Color=NamedRerouteUsage()
   Specular=NamedRerouteUsage()
   Roughness=NamedRerouteUsage()
   Opacity=NamedRerouteUsage()
   Normal=NamedRerouteUsage()
   World Position Offset=StaticSwitchParameter(True=Constant3Vector(), False=NamedRerouteUsage()) {parameterName=DisableAllAnimation}
   Subsurface Color=NamedRerouteUsage()
   Opacity Mask=StaticSwitchParameter(True=NamedRerouteUsage(), False=Constant()) {parameterName=Is2DMaskMaterial}
  ) {attributeSetTypes [
    #1
    #2
    #3
    #4
    #5
    #6
    #7
    #8
   ]}
  MaterialFunctionCall!(
   In=@n29
   UseHeightFalloff=StaticBool()
   OnFireEmissiveOverride
   HeightFalloffRadius
   HeightFalloffHardness
   VariationMasksMobile
   In_Grass=StaticBool() {value=true}
   Is_Stylised_WE=StaticBool() {value=true}
  ) {materialFunction=/Game/Packages/Fortress_SharedMaterials/Base_Material_Functions/MF_Curie_Landscape}
  n30 = StaticSwitchParameter(True=NamedRerouteUsage(), False=WorldPosition().XYZ) {defaultValue=true, group=Animation, parameterName=UseSamplingPerBladeOrWorldPosition}
  n31 = Multiply(A=NamedRerouteUsage(), B=ScalarParameter() {defaultValue=200, group=Animation, parameterName=RandomSampleOffsetPerBlade})
  n32 = MaterialFunctionCall!(
   Position=StaticSwitchParameter(True=@n30, False=Add(A=@n31, B=@n30)) {parameterName="UseRealRotationAnimation(MostExpensive)"}
   IsTrees
   IsGrass=StaticBool() {value=true}
   IsMediumPlant
   IsMediumPlant_2024Anim
   IsMediumPlant2ndLevel
   MediumPlant_BasePivotPostion
  ) {materialFunction=/Game/Environments/Asteria/Foliage/Wind/MF_WindActorData}
  NamedRerouteDeclaration(None=@n32.WindIntensitySingleSample) {name=WindIntensityNoise, nodeColor=<rgba 0.1719 0 1 1>}
  n33 = TextureSampleParameter2D() {
   group=Textures
   parameterName=SMR
   samplerSource=SSM_Wrap_WorldGroupSettings
   samplerType=SAMPLERTYPE_Masks
   texture=&T_WhiteMask
   channelNames {a="", b=Roughness, g=ColorSystemMask, r=Specular}
  }
  NamedRerouteDeclaration(None=@n33.Specular) {name=SpecularTexture, nodeColor=<rgba 0.9688 1 0 1>}
  n34 = Constant3Vector() {desc=Looks almost same RVT normal, constant=<rgba 0 0 1 0>}
  n35 = VertexInterpolator(VS=Transform(None=@n34) {transformSourceType=TRANSFORMSOURCE_Instance})
  n36 = TextureSampleParameter2D() {
   group=Textures
   parameterName=Normal
   samplerType=SAMPLERTYPE_Normal
   texture=/Game/Environments/Asteria/Foliage/Grass/Grass2DAutoBake/Grass_Standard_2D_N
  }
  n37 = MaterialFunctionCall(Normal=@n36.RGB, Flatness=ScalarParameter() {group=Normal Settings, parameterName=TextureNormalFlatten}) {materialFunction=&FlattenNormal}
  n38 = StaticSwitchParameter(
   True=@n37.Result
   False=NamedRerouteDeclaration(None=TextureSample() {samplerType=SAMPLERTYPE_Normal, texture=/Game/Environments/Asteria/Foliage/Grass/Textures/Grass/CylinderNormal_N}.RGB) {name=CylinderNormal, nodeColor=<rgba 1 0.9375 0 1>}
  ) {parameterName=Is2DMaskMaterial}
  n39 = Multiply(
   A=NamedRerouteDeclaration(None=StaticSwitchParameter(True=@n37.Result, False=@n38) {group=Textures, parameterName=UseTextures}) {name=NormalType, nodeColor=<rgba 1 0.9375 0 1>}
   B=AppendVector(A=Constant2Vector() {g=1, r=1}, B=TwoSidedSign())
  )
  n40 = Add(
   A=Multiply(A=@n34, B=ScalarParameter() {defaultValue=4, group=Normal Settings, parameterName=RVTNormalInfluence})
   B=NamedRerouteDeclaration(None=Transform(None=@n39)) {name=NormalType2, nodeColor=<rgba 1 0.9375 0 1>}
  )
  n41 = StaticSwitchParameter(True=LinearInterpolate(A=@n40, B=Constant3Vector() {constant=<rgba 0 0 1 0>}, Alpha=NamedRerouteUsage()), False=@n40) {group=FlowerFade, parameterName=UseFlowerFadeToGreen}
  NamedRerouteDeclaration(None=StaticSwitchParameter(
    True=MaterialFunctionCall(BaseNormal=@n35.PS, AdditionalNormal=@n37.Result) {materialFunction=&BlendAngleCorrectedNormals}.Result
    False=Normalize(VectorInput=@n41)
   ) {parameterName=Is2DMaskMaterial}) {name=Normal, nodeColor=<rgba 1 0 0.1406 1>}
  n42 = NamedRerouteUsage()
  n43 = Multiply(A=Multiply(A=@n18, B=LinearInterpolate(Alpha=@n42) {constA=0.3, constB=1.2}), B=NamedRerouteUsage())
  n44 = NamedRerouteUsage()
  NamedRerouteDeclaration(None=ShadingPathSwitch(Default=QualitySwitch(Default=@n43, Low=@n44, High=@n44, Medium=@n44, Epic=@n43), Mobile=@n44) {desc=Migrated from FeatureLevelSwitch}) {name=Specular, nodeColor=<rgba 0 0.1562 1 1>}
  n45 = NamedRerouteUsage()
  NamedRerouteDeclaration(None=ShadingPathSwitch(Default=QualitySwitch(Default=@n20, Low=@n45, Medium=@n45), Mobile=@n45) {desc=Migrated from FeatureLevelSwitch}) {name=Roughness, nodeColor=<rgba 0 1 0.5078 1>}
  n46 = Multiply(A=@n16, B=Constant3Vector() {constant=<rgba 0 0 -100 0>})
  NamedRerouteDeclaration(None=StaticSwitchParameter(True=Add(A=@n12, B=@n46), False=@n12) {group=Snow, parameterName=UseSnowLegacy}) {name=WPO, nodeColor=<rgba 0 1 0.3438 1>}
  n47 = LinearInterpolate(
   A=Multiply(A=NamedRerouteUsage(), B=ScalarParameter() {defaultValue=1, group=SubSurface Settings, parameterName=SSSColor})
   B=Constant()
   Alpha=NamedRerouteUsage()
  )
  n48 = NamedRerouteUsage()
  n49 = LinearInterpolate(A=ScalarParameter() {defaultValue=0.5, group=SubSurface Settings, parameterName=SSSBrightnessBase}, Alpha=@n48)
  NamedRerouteDeclaration(None=StaticSwitchParameter(True=Multiply(A=@n47, B=@n48), False=Multiply(A=@n47, B=@n49)) {parameterName=Is2DMaskMaterial}) {name=SSSColor, nodeColor=<rgba 1 0.0938 0 1>}
  n50 = ScalarParameter() {defaultValue=1, group=SubSurface Settings, parameterName=SSSOpacity}
  n51 = ScalarParameter() {defaultValue=0.5, group=SubSurface Settings, parameterName=SSSOpacity2}
  n52 = VertexColor()
  NamedRerouteDeclaration(None=Saturate(None=LinearInterpolate(A=@n50, B=@n51, Alpha=@n52.B))) {name=SSSOpacity, nodeColor=<rgba 0 0.3672 1 1>}
  n53 = NamedRerouteUsage()
  n54 = ScalarParameter() {defaultValue=0.38, group=Animation, parameterName=MaxRotation_texAnim}
  n55 = NamedRerouteUsage()
  n56 = ScalarParameter() {defaultValue=1, group=Animation, parameterName=MaxRotation_PhysicsField}
  n57 = LinearInterpolate(
   A=LinearInterpolate(A=Multiply(A=@n53, B=@n54), B=Multiply(A=@n55, B=@n56), Alpha=NamedRerouteUsage())
   B=ScalarParameter() {defaultValue=0.375, group=Animation, parameterName=MaxRotation_PlayerPush}
   Alpha=NamedRerouteUsage()
  )
  NamedRerouteDeclaration(None=@n57) {name=AnimStrength, nodeColor=<rgba 1 0.3047 0 1>}
  NamedRerouteDeclaration(None=@n1.B) {name=VerticalGradient, nodeColor=<rgba 1 0 0.0703 1>}
  NamedRerouteDeclaration(None=Frac(None=@n1.G)) {name=RandomPerBlade, nodeColor=<rgba 0.7109 0 1 1>}
  n58 = TextureCoordinate() {coordinateIndex=1}
  n59 = TransformPosition(None=MaterialFunctionCall(X=ComponentMask(None=@n58) {r=true}, Y=ComponentMask(None=@n58) {g=true}, Z=Constant()) {materialFunction=&MakeFloat3}.Result) {transformSourceType=TRANSFORMPOSSOURCE_Instance, transformType=TRANSFORMPOSSOURCE_World}
  NamedRerouteDeclaration(None=@n59) {name=PivotPosition, nodeColor=<rgba 0 0.9766 1 1>}
  n60 = TextureSampleParameter2D() {
   group=Textures
   parameterName=Mask2Dtexture
   samplerType=SAMPLERTYPE_Masks
   texture=&T_WhiteMask
   channelNames {a="", b=RandomPerBlade, g=BakedAO, r=OpacityMask}
  }
  NamedRerouteDeclaration(None=@n60.RandomPerBlade) {name=RandomPerBlade2d, nodeColor=<rgba 0.0312 1 0 1>}
  NamedRerouteDeclaration(None=TransformPosition(None=Constant3Vector()) {transformSourceType=TRANSFORMPOSSOURCE_Instance, transformType=TRANSFORMPOSSOURCE_World}) {name=InstancePosition, nodeColor=<rgba 0.3828 1 0 1>}
  n61 = CrossProduct(A=Constant3Vector() {constant=<rgba 0 0 1 0>}, B=NamedRerouteUsage())
  n62 = NamedRerouteUsage()
  n63 = LinearInterpolate(A=ScalarParameter() {defaultValue=0.5, group=Animation, parameterName=AnimBendFalloff}, Alpha=NamedRerouteUsage())
  n64 = Constant()
  n65 = TextureSample(UVs=Add(A=TextureCoordinate() {coordinateIndex=2}, B=AppendVector(A=@n57, B=@n64))) {samplerType=SAMPLERTYPE_LinearColor, texture=&/GenericShared_Grass_Foliage_Anim}
  n66 = Multiply(A=NamedRerouteUsage(), B=@n65.R)
  n67 = Multiply(
   A=Multiply(A=@n65.G, B=ScalarParameter() {defaultValue=1, group=Animation, parameterName=Z_Offset_Mult})
   B=Constant3Vector() {constant=<rgba 0 0 -1 0>}
  )
  n68 = NamedRerouteUsage()
  n69 = StaticSwitchParameter(True=LinearInterpolate(A=ScalarParameter() {defaultValue=-1, parameterName=LowPolyCounterOffset}, Alpha=@n68), False=@n68) {parameterName=UseLowPolyCounterOffset}
  n70 = Constant() {r=1}
  n71 = ScalarParameter() {defaultValue=1, parameterName=Stiffness}
  n72 = Power(Base=@n69, Exp=Clamp(None=Add(A=@n70, B=@n71)) {maxDefault=50, minDefault=1}) {constExponent=3}
  n73 = NamedRerouteUsage()
  n74 = NamedRerouteUsage()
  n75 = Multiply(
   A=Multiply(A=StaticSwitchParameter(True=@n69, False=@n72) {parameterName=IsFlowers}, B=Multiply(A=@n73, B=@n74))
   B=ScalarParameter() {defaultValue=150, parameterName=Master_Anim_strength}
  )
  n76 = StaticSwitchParameter(
   True=RotateAboutAxis(
    NormalizedRotationAxis=Normalize(VectorInput=@n61)
    RotationAngle=Multiply(A=@n62, B=@n63)
    PivotPoint=NamedRerouteUsage()
    Position=WorldPosition().XYZ
   )
   False=StaticSwitchParameter(True=Add(A=@n66, B=@n67), False=@n75) {parameterName=UseHighQualityRotationAnimation}
  ) {parameterName="UseRealRotationAnimation(MostExpensive)"}
  NamedRerouteDeclaration(None=ShadingPathSwitch(Default=QualitySwitch(Default=@n76, Low=@n75, Medium=@n75), Mobile=@n75) {desc=Migrated from FeatureLevelSwitch}) {name=Simple Lookup Animation, nodeColor=<rgba 0 1 0.8359 1>}
  NamedRerouteDeclaration(None=VertexInterpolator(VS=@n59).PS) {name=PivotForRVT, nodeColor=<rgba 1 0.75 0 1>}
  NamedRerouteDeclaration(None=@n32.WindI_Direction_WithNoise) {name=WindDirectionWithNoise, nodeColor=<rgba 0 1 0.9297 1>}
  n77 = DotProduct(A=VertexNormalWS(), B=Normalize(VectorInput=WorldPosition() {worldPositionShaderOffset=WPT_CameraRelative}.XYZ))
  n78 = Multiply(A=OneMinus(None=Abs(None=@n77)), B=ScalarParameter() {defaultValue=4, group=CameraFacingControls, parameterName=Thickening})
  NamedRerouteDeclaration(None=LinearInterpolate(
    A=MaterialFunctionCall(WidthBase=@n78, WidthTip=@n78) {materialFunction=&SplineThicken}.WorldPositionOffset
    B=Constant3Vector()
    Alpha=NamedRerouteUsage()
   )) {name=RotateGrasstoCamera, nodeColor=<rgba 0.5469 1 0 1>}
  n79 = BreakMaterialAttributes!(Attr=MaterialFunctionCall!(SetMaterialAttributes=@n27) {materialFunction=/Game/Athena/Helios/Environments/Landscape/Materials/MaterialFunctions_Utilities/MF_Decode_Grass_Biome_Colors}.Blended Material)
  n80 = MaterialFunctionCall!(BiomeMask-RGBA=@n28.BiomeMaskRGBA, Grass Localised Tint Mask, CombinedLandscapeLayersMask, BiomeColors-UseDefault) {materialFunction=/Game/Athena/Asteria/Environments/Landscape/Materials/MaterialFunctions/MF_Asteria_Common_Grass}
  n81 = VectorParameter() {group=Color, parameterName=GrassColor_FallBack, defaultValue=<rgba 0.0471 0.149 0.0471 1>}
  n82 = VectorParameter() {group=Color, parameterName=GrassColor_FallBack_Mobile, defaultValue=<rgba 0.0471 0.149 0.0471 1>}
  NamedRerouteDeclaration(None=StaticSwitchParameter(
    True=StaticSwitchParameter(True=@n79.BaseColor, False=@n80.BasicGrassMeshColor) {parameterName=UseBiomeMaskGrassColors}
    False=ShadingPathSwitch(Default=@n81.RGB, Mobile=@n82.RGB) {desc=Migrated from FeatureLevelSwitch}
   ) {defaultValue=true, parameterName="UseGrassColors(BiomeTexture)"}) {name=LandscapeColor_Influence, nodeColor=<rgba 1 0.9375 0 1>}
  NamedRerouteDeclaration(None=@n60.OpacityMask) {name=Opacity Mask, nodeColor=<rgba 0.1016 0 1 1>}
  NamedRerouteDeclaration(None=@n33.Roughness) {name=Roughness Texture, nodeColor=<rgba 0.5938 0 1 1>}
  NamedRerouteDeclaration(None=@n60.BakedAO) {name=Baked AO, nodeColor=<rgba 1 0 0.6797 1>}
  n83 = ScalarParameter() {group=Color, parameterName=BakedAOintensity}
  n84 = NamedRerouteUsage()
  n85 = StaticSwitchParameter(True=NamedRerouteUsage(), False=@n26.G) {parameterName=Is2DMaskMaterial}
  n86 = Multiply(A=NamedRerouteUsage(), B=LinearInterpolate(A=ScalarParameter() {defaultValue=1, group=Color, parameterName=ValueVar}, Alpha=@n85))
  n87 = Multiply(A=ScalarParameter() {defaultValue=-0.0832, group=Color, parameterName=HueVarPerBlade}, B=Subtract(A=@n85) {constB=0.5})
  n88 = StaticSwitchParameter(
   True=@n86
   False=MaterialFunctionCall(Hue Shift Percentage=NamedRerouteDeclaration(None=@n87) {name=HueVarPerBlade, nodeColor=<rgba 0 0.8594 1 1>}, Texture=@n86) {materialFunction=&HueShift}.Result
  ) {parameterName=IsFlowers}
  n89 = NamedRerouteUsage()
  n90 = NamedRerouteDeclaration(None=ShadingPathSwitch(Default=QualitySwitch(Default=@n88, Low=@n89, Medium=@n89), Mobile=@n89) {desc=Migrated from FeatureLevelSwitch}) {name=BaseColor, nodeColor=<rgba 0 1 0.5078 1>}
  n91 = Multiply(A=Saturate(None=LinearInterpolate(A=@n83, Alpha=@n84)), B=@n90)
  n92 = ShadingPathSwitch(
   Default=QualitySwitch(Default=StaticSwitchParameter(True=@n91, False=@n90) {parameterName=Is2DMaskMaterial}, Low=@n90, Medium=@n90)
   Mobile=@n90
  ) {desc=Migrated from FeatureLevelSwitch}
  n93 = StaticSwitchParameter(True=LinearInterpolate(A=@n92, B=NamedRerouteUsage(), Alpha=NamedRerouteUsage()), False=@n92) {group=FlowerFade, parameterName=UseFlowerFadeToGreen}
  n94 = ShadingPathSwitch(Default=@n93, Mobile=Multiply(A=@n93, B=ScalarParameter() {defaultValue=0.828, parameterName=MobileDiffuseBrightness})) {desc=Migrated from FeatureLevelSwitch}
  NamedRerouteDeclaration(None=StaticSwitchParameter(
    True=MaterialFunctionCall!(GrassMask, RockMask, DesaturationAmount, UVs, BaseColor=@n94) {materialFunction=/Game/Athena/Environments/Landscape/MaterialFunctions/Snow/MF_Athena_ColorTest_S07}.ResultWithRock
    False=@n94
   ) {parameterName=UseColorTintS07}) {name=Final BaseColor, nodeColor=<rgba 1 0 0.8203 1>}
  NamedRerouteDeclaration(None=@n25) {name=PlayerSphereMask, nodeColor=<rgba 0 0.8828 1 1>}
  NamedRerouteDeclaration(None=@n24) {name=PlayerPushDirection, nodeColor=<rgba 0 0.6016 1 1>}
  n95 = SamplePhysicsVectorField!(World Position=NamedRerouteUsage()) {fieldTarget=Vector_LinearVelocity}
  n96 = MaterialFunctionCall(Vector 3=Add(A=@n95) {constB=0.0001}) {materialFunction=/Engine/Functions/Engine_MaterialFunctions02/Utility/VectorLength}
  n97 = Constant()
  NamedRerouteDeclaration(None=ShadingPathSwitch(Default=Saturate(None=@n96.V3 Length), Mobile=@n97) {desc=Migrated from FeatureLevelSwitch}) {name=PhysicsFieldMask, nodeColor=<rgba 0 1 0.3906 1>}
  n98 = Constant3Vector() {constant=<rgba 0 0 1 0>}
  NamedRerouteDeclaration(None=ShadingPathSwitch(
    Default=MaterialFunctionCall(Vector=@n95, Default=@n98) {materialFunction=&SafeNormalize}.Result
    Mobile=Constant3Vector() {constant=<rgba 0 0 1 0>}
   ) {desc=Migrated from FeatureLevelSwitch}) {name=PhysicsFieldDirection, nodeColor=<rgba 0.3828 0 1 1>}
  n99 = Min(A=@n96.V3 Length, B=ScalarParameter() {defaultValue=1, group=Animation, parameterName=MaxRotation_PhysicsField_Clamp})
  n100 = Time()
  n101 = Multiply(A=ScalarParameter() {defaultValue=351, group=Animation, parameterName=PhysicsField_NoiseOffset}, B=NamedRerouteUsage())
  n102 = Frac(None=Multiply(A=ScalarParameter() {defaultValue=2.5, group=Animation, parameterName=MaxRotation_PhysicsField_NoiseSpeed}, B=Add(A=@n100, B=@n101)))
  n103 = OneMinus(None=@n102)
  n104 = LinearInterpolate(
   A=Multiply(A=Multiply(A=@n102, B=@n103)) {constB=2}
   Alpha=ScalarParameter() {defaultValue=0.6, group=Animation, parameterName=MaxRotation_PhysicsField_MInBend}
  )
  NamedRerouteDeclaration(None=ShadingPathSwitch(Default=Multiply(A=@n99, B=@n104), Mobile=@n97) {desc=Migrated from FeatureLevelSwitch}) {name=PhysicsFieldStrength, nodeColor=<rgba 1 0 0.2109 1>}
  n105 = TextureSampleParameter2D() {
   group=Textures
   parameterName=Diffuse
   samplerSource=SSM_Wrap_WorldGroupSettings
   texture=/Game/Characters/Fortress_Character/Textures/T_All_White
  }
  n106 = Multiply(A=NamedRerouteUsage(), B=@n105.RGB)
  n107 = NamedRerouteUsage()
  n108 = Multiply(A=Subtract(A=PerInstanceRandom()) {constB=0.5}, B=ScalarParameter() {defaultValue=-0.024, group=Color, parameterName=HueVarPerCluster})
  n109 = Multiply(A=@n105.RGB, B=VectorParameter() {group=Color, parameterName=ColorTint, defaultValue=<rgba 1 1 1 1>}.RGB)
  n110 = VectorParameter() {parameterName=SakuraBiomeFlowerColoration, defaultValue=<rgba 1.0499 1.7584 2 1>}
  n111 = Desaturation(None=@n109, Fraction=Constant() {r=0.8})
  n112 = Constant() {r=0.31}
  n113 = Constant() {r=0.37}
  n114 = Ceil(None=Subtract(A=TextureCoordinate(), B=AppendVector(A=@n112, B=@n113)))
  n115 = OneMinus(None=Max(A=ComponentMask(None=@n114) {g=true}, B=ComponentMask(None=@n114) {r=true}))
  n116 = LinearInterpolate(A=@n109, B=Saturate(None=Multiply(A=@n110.RGB, B=@n111)), Alpha=Multiply(A=Saturate(None=@n115), B=NamedRerouteUsage()))
  n117 = StaticSwitchParameter(
   True=MaterialFunctionCall(
    Hue Shift Percentage=Add(A=@n107, B=@n108)
    Texture=StaticSwitchParameter(True=@n116, False=@n109) {defaultValue=true, parameterName="UseGrassColors(BiomeTexture)"}
   ) {materialFunction=&HueShift}.Result
   False=@n105.RGB
  ) {parameterName=IsFlowers}
  n118 = NamedRerouteUsage()
  NamedRerouteDeclaration(None=StaticSwitchParameter(True=LinearInterpolate(A=@n106, B=@n117, Alpha=@n118), False=NamedRerouteUsage()) {group=Textures, parameterName=UseTextures}) {name=ColorLowQuality, nodeColor=<rgba 0.1953 0 1 1>}
  n119 = RuntimeVirtualTextureSampleParameter!(UVs, World Position=NamedRerouteUsage(), Mip Level=Constant()) {
   bEnableFeedback=false
   group=RVT
   materialType=BaseColor_Normal_Specular_YCoCg
   mipValueMode=RVTMVM_MipLevel
   parameterName=RVT Sample Target
   virtualTexture=/Game/Athena/Apollo/Maps/Landscape/VT/RVT_Apollo_Landscape
  }
  n120 = NamedRerouteUsage()
  n121 = MaterialFunctionCall(NoRVT=@n120, RVT=StaticSwitchParameter(True=@n119.BaseColor, False=@n120) {defaultValue=true, parameterName=UseRVT}) {materialFunction=&FN_RVT_Switch_Material}
  NamedRerouteDeclaration(None=ShadingPathSwitch(Default=QualitySwitch(Default=@n121.Result, Low=@n120), Mobile=@n120) {desc=Migrated from FeatureLevelSwitch}) {name=RVTorBiomeColor, nodeColor=<rgba 1 0.5391 0 1>}
  NamedRerouteDeclaration(None=@n33.ColorSystemMask) {name=ColorSytemTexturemask, nodeColor=<rgba 1 0 0.9844 1>}
  n122 = PixelDepth()
  n123 = Add(A=@n122.R, B=ScalarParameter() {defaultValue=-2000, group=Camera Fade Distances, parameterName=NearFadeOffset})
  n124 = ScalarParameter() {defaultValue=5000, group=Camera Fade Distances, parameterName=NearFade}
  NamedRerouteDeclaration(None=Saturate(None=Divide(A=@n123, B=@n124))) {name=NearFadeMask, nodeColor=<rgba 1 0 0.9609 1>}
  n125 = Add(A=@n122.R, B=ScalarParameter() {defaultValue=-7000, group=Camera Fade Distances, parameterName=FarFadeOffset})
  n126 = ScalarParameter() {defaultValue=4000, group=Camera Fade Distances, parameterName=FarFade}
  NamedRerouteDeclaration(None=Saturate(None=Divide(A=@n125, B=@n126))) {name=FarFadeMask, nodeColor=<rgba 1 0 0.8672 1>}
  n127 = Add(A=@n122.R, B=ScalarParameter() {defaultValue=-2000, group=Camera Fade Distances, parameterName=TerrainBlendOffset})
  n128 = ScalarParameter() {defaultValue=3000, group=Camera Fade Distances, parameterName=TerrainBlendFade}
  NamedRerouteDeclaration(None=Saturate(None=Divide(A=@n127, B=@n128))) {name=TerrainBlendMask, nodeColor=<rgba 1 0 0.9844 1>}
  NamedRerouteDeclaration(None=@n28.Desert) {name=SakuraBiomeMask, nodeColor=<rgba 0.5703 0 1 1>}
  n129 = LinearInterpolate(
   A=LinearInterpolate(A=NamedRerouteUsage(), B=NamedRerouteUsage(), Alpha=NamedRerouteUsage())
   B=NamedRerouteUsage()
   Alpha=NamedRerouteUsage()
  )
  n130 = MaterialFunctionCall(
   VectorToTransform=@n129
   BasisX=Transform(None=Constant3Vector() {constant=<rgba 1 0 0 0>}) {transformSourceType=TRANSFORMSOURCE_Instance}
   BasisY=Transform(None=Constant3Vector() {constant=<rgba 0 1 0 0>}) {transformSourceType=TRANSFORMSOURCE_Instance}
   BasisZ=Transform(None=Constant3Vector() {constant=<rgba 0 0 1 0>}) {transformSourceType=TRANSFORMSOURCE_Instance}
  ) {materialFunction=&Transform3x3Matrix}
  NamedRerouteDeclaration(None=Normalize(VectorInput=StaticSwitchParameter(True=@n130.Transformed Vector, False=@n129) {parameterName=UseAlignmentFix})) {name=BlendedDirectionVector, nodeColor=<rgba 0 1 0.3203 1>}
  NamedRerouteUsage()
  NamedRerouteUsage()
  n131 = ScalarParameter() {defaultValue=-3000, group=FlowerFade, parameterName=FlowerFadeOffset}
  Saturate(None=Divide(A=Add(A=@n122.R, B=@n131), B=ScalarParameter() {defaultValue=5000, group=FlowerFade, parameterName=FlowerFarFade}))
  ScalarParameter() {group=Specular, parameterName=Specular_Far}
 ]
}
